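# Skydive physics asset for the M_MED_AutumnFern male player body, covering the coat skirt and legs.
# Ten coat chains (front, front-side, side, side-back and back on each side) each hang four capsule segments from a base body constrained to root.
# Upper joints are free, while lower joints mostly lock twist and use tight, hand-tuned, asymmetric swing limits.
# Root, thigh, calf and foot capsules, linked by free constraints, supply body shapes alongside the coat.
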PhysicsAsset /BRCosmetics/Characters/Player/Male/Medium/Bodies/M_MED_AutumnFern/M_MED_AutumnFern_Skydive_Physics
bodies {
 base_coatBackSide_r [{shapeType=Capsule, length=4.071, radius=3.8636, shapeName=dyn_coatSideBack_01_r_capsule, center=<4.8734 0 0>}]
 base_coatBack_l [{shapeType=Capsule, length=4.071, radius=3.8636, shapeName=dyn_coatSideBack_01_r_capsule, center=<4.8734 0 0>}]
 base_coatBack_r [{shapeType=Capsule, length=4.071, radius=3.8636, shapeName=dyn_coatSideBack_01_r_capsule, center=<4.8734 0 0>}]
 base_coatFrontSide_l [{shapeType=Capsule, length=4.071, radius=3.8636, shapeName=dyn_coatSideBack_01_r_capsule, center=<4.8734 0 0>}]
 base_coatFrontSide_r [{shapeType=Capsule, length=4.071, radius=3.8636, shapeName=dyn_coatSideBack_01_r_capsule, center=<4.8734 0 0>}]
 base_coatFront_l [{shapeType=Capsule, length=4.071, radius=3.8636, shapeName=dyn_coatSideBack_01_r_capsule, center=<4.8734 0 0>}]
 base_coatFront_r [{shapeType=Capsule, length=4.071, radius=3.8636, shapeName=dyn_coatSideBack_01_r_capsule, center=<4.8734 0 0>}]
 base_coatSideBack_l [{shapeType=Capsule, length=4.071, radius=3.8636, shapeName=dyn_coatSideBack_01_r_capsule, center=<4.8734 0 0>}]
 base_coatSide_l [{shapeType=Capsule, length=4.071, radius=3.8636, shapeName=dyn_coatSideBack_01_r_capsule, center=<4.8734 0 0>}]
 base_coatSide_r [{shapeType=Capsule, length=4.071, radius=3.8636, shapeName=dyn_coatSideBack_01_r_capsule, center=<4.8734 0 0>}]
 calf_l [{
   shapeType=Capsule
   length=42.2271
   radius=6.674
   shapeName=calf_l_capsule
   center=<-27.0974 -1.5908 0.338>
   rotation=<rot 8.1657 -92.9841 -90.1051>
  }]
 calf_r [{
   shapeType=Capsule
   length=44.582
   radius=6.674
   shapeName=calf_r_capsule
   center=<27.0974 1.5908 -0.338>
   rotation=<rot 8.1657 -92.9841 89.8949>
  }]
 dyn_coatBack_01_l [{shapeType=Capsule, length=4.071, radius=3.8636, shapeName=dyn_coatSideBack_01_r_capsule, center=<4.8734 0 0>}]
 dyn_coatBack_01_r [{shapeType=Capsule, length=4.071, radius=3.8636, shapeName=dyn_coatSideBack_01_r_capsule, center=<4.8734 0 0>}]
 dyn_coatBack_02_l [{
   shapeType=Capsule
   length=7.4207
   radius=9.0745
   shapeName=dyn_coatSideBack_01_r_capsule
   center=<3.6988 -4.9405 0>
   rotation=<rot 0 -29.8209 0>
  }]
 dyn_coatBack_02_r [{shapeType=Capsule, length=7.4207, radius=9.0745, shapeName=dyn_coatSideBack_01_r_capsule, center=<6.3112 -4.2864 0>}]
 dyn_coatBack_03_l [{
   shapeType=Capsule
   length=9.1718
   radius=7.0869
   shapeName=dyn_coatSideBack_01_r_capsule
   center=<8.7881 -1.0213 -1.0015>
   rotation=<rot -3.4265 0 0>
  }]
 dyn_coatBack_03_r [{
   shapeType=Capsule
   length=11.9046
   radius=7.0869
   shapeName=dyn_coatSideBack_01_r_capsule
   center=<8.7881 -1.227 -1.0015>
   rotation=<rot -2.9723 -29.8652 1.7055>
  }]
 dyn_coatBack_04_l [{
   shapeType=Capsule
   length=10.4572
   radius=7.0869
   shapeName=dyn_coatSideBack_01_r_capsule
   center=<8.7881 -1.0213 -1.0015>
   rotation=<rot -3.4265 0 0>
  }]
 dyn_coatBack_04_r [{
   shapeType=Capsule
   length=9.1718
   radius=7.0869
   shapeName=dyn_coatSideBack_01_r_capsule
   center=<8.7881 -1.0213 -1.0015>
   rotation=<rot -3.4265 0 0>
  }]
 dyn_coatFrontSide_01_l [{shapeType=Capsule, length=4.071, radius=3.8636, shapeName=dyn_coatSideBack_01_r_capsule, center=<4.8734 0 0>}]
 dyn_coatFrontSide_01_r [{shapeType=Capsule, length=4.071, radius=3.8636, shapeName=dyn_coatSideBack_01_r_capsule, center=<4.8734 0 0>}]
 dyn_coatFrontSide_02_l [{shapeType=Capsule, length=4.071, radius=3.8636, shapeName=dyn_coatSideBack_01_r_capsule, center=<4.8734 0 0>}]
 dyn_coatFrontSide_02_r [{shapeType=Capsule, length=4.071, radius=3.8636, shapeName=dyn_coatSideBack_01_r_capsule, center=<4.8734 0 0>}]
 dyn_coatFrontSide_03_l [{
   shapeType=Capsule
   length=11.9046
   radius=7.0869
   shapeName=dyn_coatSideBack_01_r_capsule
   center=<8.7881 -4.5393 -1.0015>
   rotation=<rot -3.4265 0 0>
  }]
 dyn_coatFrontSide_03_r [{
   shapeType=Capsule
   length=10.6961
   radius=7.0869
   shapeName=dyn_coatSideBack_01_r_capsule
   center=<8.8465 -6.9349 0>
   rotation=<rot -3.4265 0 -4.4978>
  }]
 dyn_coatFrontSide_04_l [{
   shapeType=Capsule
   length=13.1678
   radius=7.6615
   shapeName=dyn_coatSideBack_01_r_capsule
   center=<10.7047 -5.2402 -1.6136>
   rotation=<rot 0.0001 -0.0001 9.2593>
  }]
 dyn_coatFrontSide_04_r [{
   shapeType=Capsule
   length=10.6961
   radius=7.0869
   shapeName=dyn_coatSideBack_01_r_capsule
   center=<8.8465 -6.9349 0>
   rotation=<rot -3.4265 0 -4.4978>
  }]
 dyn_coatFront_01_l [{shapeType=Capsule, length=4.071, radius=3.8636, shapeName=dyn_coatSideBack_01_r_capsule, center=<4.8734 0 0>}]
 dyn_coatFront_01_r [{shapeType=Capsule, length=4.071, radius=3.8636, shapeName=dyn_coatSideBack_01_r_capsule, center=<4.8734 0 0>}]
 dyn_coatFront_02_l [{shapeType=Capsule, length=4.071, radius=3.8636, shapeName=dyn_coatSideBack_01_r_capsule, center=<12.8653 -0.832 0>}]
 dyn_coatFront_02_r [{shapeType=Capsule, length=4.071, radius=3.8636, shapeName=dyn_coatSideBack_01_r_capsule, center=<4.8734 0 0>}]
 dyn_coatFront_03_l [{
   shapeType=Capsule
   length=11.9046
   radius=5.4378
   shapeName=dyn_coatSideBack_01_r_capsule
   center=<8.7881 -2.9683 -1>
   rotation=<rot -3.4265 0 0>
  }]
 dyn_coatFront_03_r [{
   shapeType=Capsule
   length=11.9046
   radius=7.0869
   shapeName=dyn_coatSideBack_01_r_capsule
   center=<5.9156 -5.8664 -1.0015>
   rotation=<rot -3.4265 0 -0.1959>
  }]
 dyn_coatFront_04_l [{
   shapeType=Capsule
   length=10.3061
   radius=6.9631
   shapeName=dyn_coatSideBack_01_r_capsule
   center=<9.7272 -1.8877 -0.451>
   rotation=<rot 0 0 -8.481>
  }]
 dyn_coatFront_04_r [{shapeType=Capsule, length=7.3422, radius=7.7174, shapeName=dyn_coatSideBack_01_r_capsule, center=<11.1917 -6.0213 0>}]
 dyn_coatSideBack_01_l [{shapeType=Capsule, length=4.071, radius=3.8636, shapeName=dyn_coatSideBack_01_r_capsule, center=<4.8734 0 0>}]
 dyn_coatSideBack_01_r [{shapeType=Capsule, length=4.071, radius=3.8636, shapeName=dyn_coatSideBack_01_r_capsule, center=<4.8734 0 0>}]
 dyn_coatSideBack_02_l [{shapeType=Capsule, length=8.8967, radius=5.9783, shapeName=dyn_coatSideBack_01_r_capsule, center=<6.8056 -1.1969 -1.4638>}]
 dyn_coatSideBack_02_r [{shapeType=Capsule, length=8.8967, radius=5.9783, shapeName=dyn_coatSideBack_01_r_capsule, center=<5.8236 -1.8785 -1.4638>}]
 dyn_coatSideBack_03_l [{
   shapeType=Capsule
   length=9.1718
   radius=7.0869
   shapeName=dyn_coatSideBack_01_r_capsule
   center=<8.7881 -1.3767 -1.0015>
   rotation=<rot -3.4265 0 0>
  }]
 dyn_coatSideBack_03_r [{
   shapeType=Capsule
   length=11.9046
   radius=7.0869
   shapeName=dyn_coatSideBack_01_r_capsule
   center=<8.7881 -1.227 -1.0015>
   rotation=<rot -3.4265 0 0>
  }]
 dyn_coatSideBack_04_l [{
   shapeType=Capsule
   length=12.4174
   radius=7.0869
   shapeName=dyn_coatSideBack_01_r_capsule
   center=<8.7881 -1.3767 -1.0015>
   rotation=<rot -3.4265 0.0001 -16.0941>
  }]
 dyn_coatSideBack_04_r [{
   shapeType=Capsule
   length=11.9046
   radius=7.0869
   shapeName=dyn_coatSideBack_01_r_capsule
   center=<8.7881 -1.227 -1.0015>
   rotation=<rot -3.4265 0 0>
  }]
 dyn_coatSide_01_l [{shapeType=Capsule, length=4.071, radius=3.8636, shapeName=dyn_coatSideBack_01_r_capsule, center=<4.8734 0 0>}]
 dyn_coatSide_01_r [{shapeType=Capsule, length=4.071, radius=3.8636, shapeName=dyn_coatSideBack_01_r_capsule, center=<4.8734 0 0>}]
 dyn_coatSide_02_l [{shapeType=Capsule, length=4.071, radius=3.8636, shapeName=dyn_coatSideBack_01_r_capsule, center=<4.8734 0 0>}]
 dyn_coatSide_02_r [{shapeType=Capsule, length=4.071, radius=3.8636, shapeName=dyn_coatSideBack_01_r_capsule, center=<4.8734 0 0>}]
 dyn_coatSide_03_l [{
   shapeType=Capsule
   length=11.5649
   radius=7.0869
   shapeName=dyn_coatSideBack_01_r_capsule
   center=<8.7881 -3.3166 -1.0015>
   rotation=<rot -3.4265 0 0>
  }]
 dyn_coatSide_03_r [{
   shapeType=Capsule
   length=11.9046
   radius=7.0869
   shapeName=dyn_coatSideBack_01_r_capsule
   center=<8.7881 -3.5475 -1.0015>
   rotation=<rot -3.4265 0 0>
  }]
 dyn_coatSide_04_l [{
   shapeType=Capsule
   length=15.2572
   radius=7.0661
   shapeName=dyn_coatSideBack_01_r_capsule
   center=<8.7881 -3.5416 -1.0015>
   rotation=<rot -3.4265 0 0>
  }]
 dyn_coatSide_04_r [{
   shapeType=Capsule
   length=11.2519
   radius=7.0869
   shapeName=dyn_coatSideBack_01_r_capsule
   center=<8.7881 -3.4809 -1.0015>
   rotation=<rot -3.4265 0 0>
  }]
 foot_l [{
   shapeType=Capsule
   length=15.9318
   radius=5.4213
   shapeName=foot_l_capsule
   center=<-6.3834 6.2651 -0.946>
   rotation=<rot -0.7729 2.0773 -84.6694>
  }]
 foot_r [{
   shapeType=Capsule
   length=10.482
   radius=5.4213
   shapeName=foot_r_capsule
   center=<5.2354 2.456 1.4822>
   rotation=<rot -0.7729 2.0773 176.569>
  }]
 root [
  {shapeType=Capsule, length=74.1652, radius=140.5603, shapeName=root_capsule, center=<0 -19.9457 -142.3118>, rotation=<rot 0 -180 -90>}
  {shapeType=Capsule, length=0, radius=32.9404, shapeName=root_capsule, center=<0 -3.5764 -14.9343>}
 ]
 thigh_l [
  {
   shapeType=Capsule
   length=41.677
   radius=6.1574
   shapeName=thigh_l_capsule
   center=<-16.8598 1.0217 1.1886>
   rotation=<rot 81.296 112.6982 112.4778>
  }
  {
   shapeType=Capsule
   length=7.4266
   radius=6.5113
   shapeName=thigh_l_capsule
   center=<-1.4598 -3.2747 2.3682>
   rotation=<rot 88.5494 179.9982 6.605>
  }
 ]
 thigh_r [
  {
   shapeType=Capsule
   length=41.677
   radius=6.1574
   shapeName=thigh_r_capsule
   center=<16.8598 -1.0217 -1.1886>
   rotation=<rot 81.296 112.6982 -67.5222>
  }
  {
   shapeType=Capsule
   length=7.4266
   radius=6.5113
   shapeName=thigh_r_capsule
   center=<1.4598 3.2747 -2.3682>
   rotation=<rot 88.5494 179.9982 -173.395>
  }
 ]
}
constraints [
 {
  bone1Name=dyn_coatSideBack_01_r
  bone2Name=base_coatBackSide_r
  swing1LimitDegrees=45
  swing1Motion=Free
  swing2LimitDegrees=45
  swing2Motion=Free
  twistLimitDegrees=45
  twistMotion=Free
 }
 {
  bone1Name=dyn_coatSideBack_02_r
  bone2Name=dyn_coatSideBack_01_r
  swing1LimitDegrees=10
  swing2LimitDegrees=2
  swing2Motion=Locked
  twistLimitDegrees=45
  twistMotion=Locked
 }
 {
  bone1Name=dyn_coatSideBack_03_r
  bone2Name=dyn_coatSideBack_02_r
  swing1LimitDegrees=15
  swing2LimitDegrees=2
  twistLimitDegrees=45
  twistMotion=Locked
 }
 {
  bone1Name=dyn_coatSideBack_04_r
  bone2Name=dyn_coatSideBack_03_r
  swing1LimitDegrees=20
  swing2LimitDegrees=2
  twistLimitDegrees=45
  twistMotion=Locked
 }
 {
  bone1Name=dyn_coatBack_01_l
  bone2Name=base_coatBack_l
  swing1LimitDegrees=45
  swing1Motion=Free
  swing2LimitDegrees=45
  swing2Motion=Free
  twistLimitDegrees=45
  twistMotion=Free
 }
 {
  bone1Name=dyn_coatBack_02_l
  bone2Name=dyn_coatBack_01_l
  swing1LimitDegrees=10
  swing2LimitDegrees=2
  swing2Motion=Locked
  twistLimitDegrees=45
  twistMotion=Locked
 }
 {
  bone1Name=dyn_coatBack_03_l
  bone2Name=dyn_coatBack_02_l
  swing1LimitDegrees=16
  swing2LimitDegrees=2
  twistLimitDegrees=45
  twistMotion=Locked
 }
 {
  bone1Name=dyn_coatBack_04_l
  bone2Name=dyn_coatBack_03_l
  swing1LimitDegrees=16
  swing2LimitDegrees=2
  twistLimitDegrees=45
  twistMotion=Locked
 }
 {
  bone1Name=dyn_coatBack_01_r
  bone2Name=base_coatBack_r
  swing1LimitDegrees=45
  swing1Motion=Free
  swing2LimitDegrees=45
  swing2Motion=Free
  twistLimitDegrees=45
  twistMotion=Free
 }
 {
  bone1Name=dyn_coatBack_02_r
  bone2Name=dyn_coatBack_01_r
  swing1LimitDegrees=10
  swing2LimitDegrees=1
  twistLimitDegrees=45
  twistMotion=Locked
 }
 {
  bone1Name=dyn_coatBack_03_r
  bone2Name=dyn_coatBack_02_r
  swing1LimitDegrees=16
  swing2LimitDegrees=2
  twistLimitDegrees=45
  twistMotion=Locked
 }
 {
  bone1Name=dyn_coatBack_04_r
  bone2Name=dyn_coatBack_03_r
  swing1LimitDegrees=16
  swing2LimitDegrees=2
  twistLimitDegrees=45
  twistMotion=Locked
 }
 {
  bone1Name=dyn_coatFrontSide_01_l
  bone2Name=base_coatFrontSide_l
  swing1LimitDegrees=45
  swing1Motion=Free
  swing2LimitDegrees=45
  swing2Motion=Free
  twistLimitDegrees=45
  twistMotion=Free
 }
 {
  bone1Name=dyn_coatFrontSide_02_l
  bone2Name=dyn_coatFrontSide_01_l
  swing1LimitDegrees=45
  swing1Motion=Free
  swing2LimitDegrees=45
  swing2Motion=Free
  twistLimitDegrees=45
  twistMotion=Free
 }
 {
  bone1Name=dyn_coatFrontSide_03_l
  bone2Name=dyn_coatFrontSide_02_l
  swing1LimitDegrees=23
  swing2LimitDegrees=8
  twistLimitDegrees=45
  twistMotion=Locked
 }
 {
  bone1Name=dyn_coatFrontSide_04_l
  bone2Name=dyn_coatFrontSide_03_l
  swing1LimitDegrees=23
  swing2LimitDegrees=8
  twistLimitDegrees=45
  twistMotion=Locked
 }
 {
  bone1Name=dyn_coatFrontSide_01_r
  bone2Name=base_coatFrontSide_r
  swing1LimitDegrees=45
  swing1Motion=Free
  swing2LimitDegrees=45
  swing2Motion=Free
  twistLimitDegrees=45
  twistMotion=Free
 }
 {
  bone1Name=dyn_coatFrontSide_02_r
  bone2Name=dyn_coatFrontSide_01_r
  swing1LimitDegrees=45
  swing1Motion=Free
  swing2LimitDegrees=45
  swing2Motion=Free
  twistLimitDegrees=45
  twistMotion=Free
 }
 {
  bone1Name=dyn_coatFrontSide_03_r
  bone2Name=dyn_coatFrontSide_02_r
  swing1LimitDegrees=20
  swing2LimitDegrees=3
  twistLimitDegrees=45
  twistMotion=Locked
 }
 {
  bone1Name=dyn_coatFrontSide_04_r
  bone2Name=dyn_coatFrontSide_03_r
  swing1LimitDegrees=20
  swing2LimitDegrees=3
  twistLimitDegrees=45
  twistMotion=Locked
 }
 {
  bone1Name=dyn_coatFront_01_l
  bone2Name=base_coatFront_l
  swing1LimitDegrees=45
  swing1Motion=Free
  swing2LimitDegrees=45
  swing2Motion=Free
  twistLimitDegrees=45
  twistMotion=Free
 }
 {
  bone1Name=dyn_coatFront_02_l
  bone2Name=dyn_coatFront_01_l
  swing1LimitDegrees=6
  swing2LimitDegrees=2
  twistLimitDegrees=45
  twistMotion=Locked
 }
 {
  bone1Name=dyn_coatFront_03_l
  bone2Name=dyn_coatFront_02_l
  swing1LimitDegrees=40
  swing2LimitDegrees=5
  twistLimitDegrees=45
  twistMotion=Locked
 }
 {
  bone1Name=dyn_coatFront_04_l
  bone2Name=dyn_coatFront_03_l
  swing1LimitDegrees=32
  swing2LimitDegrees=5
  twistLimitDegrees=45
  twistMotion=Locked
 }
 {
  bone1Name=dyn_coatFront_01_r
  bone2Name=base_coatFront_r
  swing1LimitDegrees=45
  swing1Motion=Free
  swing2LimitDegrees=45
  swing2Motion=Free
  twistLimitDegrees=45
  twistMotion=Free
 }
 {
  bone1Name=dyn_coatFront_02_r
  bone2Name=dyn_coatFront_01_r
  swing1LimitDegrees=45
  swing1Motion=Free
  swing2LimitDegrees=45
  swing2Motion=Free
  twistLimitDegrees=45
  twistMotion=Free
 }
 {
  bone1Name=dyn_coatFront_03_r
  bone2Name=dyn_coatFront_02_r
  swing1LimitDegrees=15
  swing2LimitDegrees=10
  twistLimitDegrees=45
  twistMotion=Locked
 }
 {
  bone1Name=dyn_coatFront_04_r
  bone2Name=dyn_coatFront_03_r
  swing1LimitDegrees=25
  swing2LimitDegrees=20
  twistLimitDegrees=45
  twistMotion=Locked
 }
 {
  bone1Name=dyn_coatSideBack_01_l
  bone2Name=base_coatSideBack_l
  swing1LimitDegrees=45
  swing1Motion=Free
  swing2LimitDegrees=45
  swing2Motion=Free
  twistLimitDegrees=45
  twistMotion=Free
 }
 {
  bone1Name=dyn_coatSideBack_02_l
  bone2Name=dyn_coatSideBack_01_l
  swing1LimitDegrees=6
  swing2LimitDegrees=2
  swing2Motion=Locked
  twistLimitDegrees=45
  twistMotion=Locked
 }
 {
  bone1Name=dyn_coatSideBack_03_l
  bone2Name=dyn_coatSideBack_02_l
  swing1LimitDegrees=15
  swing2LimitDegrees=5
  twistLimitDegrees=45
  twistMotion=Locked
 }
 {
  bone1Name=dyn_coatSideBack_04_l
  bone2Name=dyn_coatSideBack_03_l
  swing1LimitDegrees=15
  swing2LimitDegrees=5
  twistLimitDegrees=45
  twistMotion=Locked
 }
 {
  bone1Name=dyn_coatSide_01_l
  bone2Name=base_coatSide_l
  swing1LimitDegrees=45
  swing1Motion=Free
  swing2LimitDegrees=45
  swing2Motion=Free
  twistLimitDegrees=45
  twistMotion=Free
 }
 {
  bone1Name=dyn_coatSide_02_l
  bone2Name=dyn_coatSide_01_l
  swing1LimitDegrees=20
  swing2LimitDegrees=10
  twistLimitDegrees=45
  twistMotion=Locked
 }
 {
  bone1Name=dyn_coatSide_03_l
  bone2Name=dyn_coatSide_02_l
  swing1LimitDegrees=20
  swing2LimitDegrees=10
  twistLimitDegrees=45
  twistMotion=Locked
 }
 {
  bone1Name=dyn_coatSide_04_l
  bone2Name=dyn_coatSide_03_l
  swing1LimitDegrees=16
  swing2LimitDegrees=10
  twistLimitDegrees=45
  twistMotion=Locked
 }
 {
  bone1Name=dyn_coatSide_01_r
  bone2Name=base_coatSide_r
  swing1LimitDegrees=45
  swing1Motion=Free
  swing2LimitDegrees=45
  swing2Motion=Free
  twistLimitDegrees=45
  twistMotion=Free
 }
 {
  bone1Name=dyn_coatSide_02_r
  bone2Name=dyn_coatSide_01_r
  swing1LimitDegrees=20
  swing2LimitDegrees=2
  swing2Motion=Locked
  twistLimitDegrees=45
  twistMotion=Locked
 }
 {
  bone1Name=dyn_coatSide_03_r
  bone2Name=dyn_coatSide_02_r
  swing1LimitDegrees=10
  swing2LimitDegrees=2
  twistLimitDegrees=45
  twistMotion=Locked
 }
 {
  bone1Name=dyn_coatSide_04_r
  bone2Name=dyn_coatSide_03_r
  swing1LimitDegrees=10
  swing2LimitDegrees=2
  twistLimitDegrees=45
  twistMotion=Locked
 }
 {
  bone1Name=calf_l
  bone2Name=thigh_l
  swing1LimitDegrees=45
  swing1Motion=Free
  swing2LimitDegrees=45
  swing2Motion=Free
  twistLimitDegrees=45
  twistMotion=Free
 }
 {
  bone1Name=calf_r
  bone2Name=thigh_r
  swing1LimitDegrees=45
  swing1Motion=Free
  swing2LimitDegrees=45
  swing2Motion=Free
  twistLimitDegrees=45
  twistMotion=Free
 }
 {
  bone1Name=foot_l
  bone2Name=calf_l
  swing1LimitDegrees=45
  swing1Motion=Free
  swing2LimitDegrees=45
  swing2Motion=Free
  twistLimitDegrees=45
  twistMotion=Free
 }
 {
  bone1Name=foot_r
  bone2Name=calf_r
  swing1LimitDegrees=45
  swing1Motion=Free
  swing2LimitDegrees=45
  swing2Motion=Free
  twistLimitDegrees=45
  twistMotion=Free
 }
 {
  bone1Name=thigh_l
  bone2Name=root
  swing1LimitDegrees=45
  swing1Motion=Free
  swing2LimitDegrees=45
  swing2Motion=Free
  twistLimitDegrees=45
  twistMotion=Free
 }
 {
  bone1Name=thigh_r
  bone2Name=root
  swing1LimitDegrees=45
  swing1Motion=Free
  swing2LimitDegrees=45
  swing2Motion=Free
  twistLimitDegrees=45
  twistMotion=Free
 }
 {
  bone1Name=base_coatFront_l
  bone2Name=root
  swing1LimitDegrees=45
  swing1Motion=Free
  swing2LimitDegrees=45
  swing2Motion=Free
  twistLimitDegrees=45
  twistMotion=Free
 }
 {
  bone1Name=base_coatFrontSide_l
  bone2Name=root
  swing1LimitDegrees=45
  swing1Motion=Free
  swing2LimitDegrees=45
  swing2Motion=Free
  twistLimitDegrees=45
  twistMotion=Free
 }
 {
  bone1Name=base_coatSide_l
  bone2Name=root
  swing1LimitDegrees=45
  swing1Motion=Free
  swing2LimitDegrees=45
  swing2Motion=Free
  twistLimitDegrees=45
  twistMotion=Free
 }
 {
  bone1Name=base_coatSideBack_l
  bone2Name=root
  swing1LimitDegrees=45
  swing1Motion=Free
  swing2LimitDegrees=45
  swing2Motion=Free
  twistLimitDegrees=45
  twistMotion=Free
 }
 {
  bone1Name=base_coatBack_r
  bone2Name=root
  swing1LimitDegrees=45
  swing1Motion=Free
  swing2LimitDegrees=45
  swing2Motion=Free
  twistLimitDegrees=45
  twistMotion=Free
 }
 {
  bone1Name=base_coatBack_l
  bone2Name=root
  swing1LimitDegrees=45
  swing1Motion=Free
  swing2LimitDegrees=45
  swing2Motion=Free
  twistLimitDegrees=45
  twistMotion=Free
 }
 {
  bone1Name=base_coatBackSide_r
  bone2Name=root
  swing1LimitDegrees=45
  swing1Motion=Free
  swing2LimitDegrees=45
  swing2Motion=Free
  twistLimitDegrees=45
  twistMotion=Free
 }
 {
  bone1Name=base_coatSide_r
  bone2Name=root
  swing1LimitDegrees=45
  swing1Motion=Free
  swing2LimitDegrees=45
  swing2Motion=Free
  twistLimitDegrees=45
  twistMotion=Free
 }
 {
  bone1Name=base_coatFrontSide_r
  bone2Name=root
  swing1LimitDegrees=45
  swing1Motion=Free
  swing2LimitDegrees=45
  swing2Motion=Free
  twistLimitDegrees=45
  twistMotion=Free
 }
 {
  bone1Name=base_coatFront_r
  bone2Name=root
  swing1LimitDegrees=45
  swing1Motion=Free
  swing2LimitDegrees=45
  swing2Motion=Free
  twistLimitDegrees=45
  twistMotion=Free
 }
]
