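# Post-process-style cosmetic AnimBP for the AutumnFern male medium body, adding coat secondary motion to the pose copied from the attached parent mesh.
# A Control Rig driven by movement flags feeds two RigidBody branches, ground and skydive (gravity-overridden), each followed by additive coat and glute bone corrections.
# A bool blend selects the skydive branch during skydiving, emotes or knee slides.
# The event graph updates state flags, weights and parameter-asset gravity override.
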AnimBlueprint /BRCosmetics/Characters/Player/Male/Medium/Bodies/M_MED_AutumnFern/M_MED_AutumnFern_AnimBP
target_skeleton=&/M_MED_AutumnFern_Skeleton
graphs [
 AnimGraph {nodes [
   "Transform (Modify) Bone - Bone: dyn_coatFront_03_r"!(ComponentPose, bAlphaBoolEnabled:bool=True, Translation="0,0,0", Rotation="0,10,0", Scale="1,1,1") -> (Pose) {
    alphaInputType=Bool
    rotationMode=BMM_Additive
    rotationSpace=BCS_BoneSpace
    translationMode=BMM_Additive
    translationSpace=BCS_BoneSpace
    alphaBoolBlend {blendInTime=0.1, blendOption=Linear, blendOutTime=0.1}
    boneToModify {boneName=dyn_coatFront_03_r}
    rotation {yaw=10}
   }
   Output Pose(Result=Blend Poses by bool(
     BlendPose_0=Use cached pose 'SkydivingPose'()
     BlendPose_1=Use cached pose 'GroundPose'()
     bActiveValue=Get isSkydivingOrEmoting() -> (isSkydivingOrEmoting:bool)
    ) {blendType=Linear}) {layerGroup=None}
   n1 = RigidBody(ComponentPose=Local To Component(LocalPose=Use cached pose 'CR_Pose'())) {
    overridePhysicsAsset=&/M_MED_AutumnFern_Physics
    simulationSpace=BaseBoneSpace
    alphaBoolBlend {blendInTime=2, blendOption=Linear, blendOutTime=0.5}
    alphaScaleBiasClamp {bMapRange=true, outRange {min=0.6}}
    baseBoneRef {boneName=root}
    componentAppliedLinearAccClamp=<4000 4000 4000>
    simSpaceSettings {
     dampingAlpha=0
     maxAngularAcceleration=12
     maxAngularVelocity=5
     maxLinearAcceleration=900
     maxLinearVelocity=300
     velocityScaleZ=0.4
     worldAlpha=0.5
    }
   }
   n2 = "Transform (Modify) Bone - Bone: base_coatFront_l"!(ComponentPose=@n1, bAlphaBoolEnabled:bool=True, Translation="-0.097,-1.498,1.307", Rotation="0.005,-3.313,-0.018", Scale="1,1,1") -> (Pose) {
    alphaInputType=Bool
    rotationMode=BMM_Additive
    rotationSpace=BCS_BoneSpace
    translationMode=BMM_Additive
    translationSpace=BCS_BoneSpace
    alphaBoolBlend {blendInTime=0.1, blendOption=Linear, blendOutTime=0.1}
    boneToModify {boneName=base_coatFront_l}
    rotation=<rot 0.005 -3.313 -0.018>
    translation=<-0.097 -1.498 1.307>
   }
   n3 = "Transform (Modify) Bone - Bone: base_coatFrontSide_l"!(ComponentPose=@n2, bAlphaBoolEnabled:bool=True, Translation="-0.097,-1.498,1.307", Rotation="-0.002,-5.751,-0.015", Scale="1,1,1") -> (Pose) {
    alphaInputType=Bool
    rotationMode=BMM_Additive
    rotationSpace=BCS_BoneSpace
    translationMode=BMM_Additive
    translationSpace=BCS_BoneSpace
    alphaBoolBlend {blendInTime=0.1, blendOption=Linear, blendOutTime=0.1}
    boneToModify {boneName=base_coatFrontSide_l}
    rotation=<rot -0.002 -5.751 -0.015>
    translation=<-0.097 -1.498 1.307>
   }
   n4 = "Transform (Modify) Bone - Bone: base_coatFront_r"!(ComponentPose=@n3, bAlphaBoolEnabled:bool=True, Translation="0,0,0", Rotation="-0.002,-3.532,-0.014", Scale="1,1,1") -> (Pose) {
    alphaInputType=Bool
    rotationMode=BMM_Additive
    rotationSpace=BCS_BoneSpace
    translationMode=BMM_Additive
    translationSpace=BCS_BoneSpace
    alphaBoolBlend {blendInTime=0.1, blendOption=Linear, blendOutTime=0.1}
    boneToModify {boneName=base_coatFront_r}
    rotation=<rot -0.002 -3.5321 -0.014>
   }
   n5 = "Transform (Modify) Bone - Bone: base_coatFrontSide_r"!(ComponentPose=@n4, bAlphaBoolEnabled:bool=True, Translation="-0.097,-1.498,1.307", Rotation="-0.002,-3.532,-0.014", Scale="1,1,1") -> (Pose) {
    alphaInputType=Bool
    rotationMode=BMM_Additive
    rotationSpace=BCS_BoneSpace
    translationMode=BMM_Additive
    translationSpace=BCS_BoneSpace
    alphaBoolBlend {blendInTime=0.1, blendOption=Linear, blendOutTime=0.1}
    boneToModify {boneName=base_coatFrontSide_r}
    rotation=<rot -0.002 -3.5321 -0.014>
    translation=<-0.097 -1.498 1.307>
   }
   n6 = "Transform (Modify) Bone - Bone: deform_glute_r"!(ComponentPose=@n5, bAlphaBoolEnabled:bool=True, Translation="-1.794,-1.670,-0.008", Rotation="0,0,0", Scale="1,1,1") -> (Pose) {
    alphaInputType=Bool
    rotationMode=BMM_Additive
    rotationSpace=BCS_BoneSpace
    translationMode=BMM_Additive
    translationSpace=BCS_BoneSpace
    alphaBoolBlend {blendInTime=0.3, blendOption=Linear, blendOutTime=0.3}
    boneToModify {boneName=deform_glute_r}
    translation=<-1.794 -1.67 -0.008>
   }
   n7 = "Transform (Modify) Bone - Bone: deform_glute_l"!(ComponentPose=@n6, bAlphaBoolEnabled:bool=True, Translation="1.632,1.287,0.000", Rotation="0,0,0", Scale="1,1,1") -> (Pose) {
    alphaInputType=Bool
    rotationMode=BMM_Additive
    rotationSpace=BCS_BoneSpace
    translationMode=BMM_Additive
    translationSpace=BCS_BoneSpace
    alphaBoolBlend {blendInTime=0.3, blendOption=Linear, blendOutTime=0.3}
    boneToModify {boneName=deform_glute_l}
    translation=<1.632 1.287>
   }
   n8 = "Transform (Modify) Bone - Bone: dyn_coatBack_03_r"!(ComponentPose=@n7, bAlphaBoolEnabled:bool=True, Translation="0,0,0", Rotation="0,-9,0", Scale="1,1,1") -> (Pose) {
    alphaInputType=Bool
    rotationMode=BMM_Additive
    rotationSpace=BCS_BoneSpace
    translationMode=BMM_Additive
    translationSpace=BCS_BoneSpace
    alphaBoolBlend {blendInTime=0.3, blendOption=Linear, blendOutTime=0.3}
    boneToModify {boneName=dyn_coatBack_03_r}
    rotation {yaw=-9}
   }
   n9 = "Transform (Modify) Bone - Bone: dyn_coatSideBack_03_r"!(ComponentPose=@n8, bAlphaBoolEnabled:bool=True, Translation="0,0,0", Rotation="2.448,-8.663,15.665", Scale="1,1,1") -> (Pose) {
    alphaInputType=Bool
    rotationMode=BMM_Additive
    rotationSpace=BCS_BoneSpace
    translationMode=BMM_Additive
    translationSpace=BCS_BoneSpace
    alphaBoolBlend {blendInTime=0.3, blendOption=Linear, blendOutTime=0.3}
    boneToModify {boneName=dyn_coatSideBack_03_r}
    rotation=<rot 2.448 -8.663 15.665>
   }
   n10 = "Transform (Modify) Bone - Bone: base_coatFront_l"(ComponentPose=@n9, bAlphaBoolEnabled=True, Translation="-0.097,-1.498,1.307", Rotation="-0.002,-3.532,-0.014", Scale="1,1,1") {
    alphaInputType=Bool
    rotationMode=BMM_Additive
    rotationSpace=BCS_BoneSpace
    translationMode=BMM_Additive
    translationSpace=BCS_BoneSpace
    alphaBoolBlend {blendInTime=0.1, blendOption=Linear, blendOutTime=0.1}
    boneToModify {boneName=base_coatFront_l}
    rotation=<rot -0.002 -3.5321 -0.014>
    translation=<-0.097 -1.498 1.307>
   }
   n11 = "Transform (Modify) Bone - Bone: base_coatFront_r"(ComponentPose=@n10, bAlphaBoolEnabled=True, Translation="-0.097,-1.498,1.307", Rotation="-0.002,-3.532,-0.014", Scale="1,1,1") {
    alphaInputType=Bool
    rotationMode=BMM_Additive
    rotationSpace=BCS_BoneSpace
    translationMode=BMM_Additive
    translationSpace=BCS_BoneSpace
    alphaBoolBlend {blendInTime=0.1, blendOption=Linear, blendOutTime=0.1}
    boneToModify {boneName=base_coatFront_r}
    rotation=<rot -0.002 -3.5321 -0.014>
    translation=<-0.097 -1.498 1.307>
   }
   Save cached pose 'GroundPose'(Pose=Component To Local(ComponentPose=@n11)) {cacheName=GroundPose}
   n12 = Control Rig!(
    Source=Copy Pose From Mesh() {bUseAttachedParent=true}
    Alpha
    isCrouching_CR:bool=Get bIsCrouching()
    isBaseballSliding_CR:bool=Get bIsBaseballSliding()
    isKneeSliding_CR:bool=Get bIsKneeSliding()
    isSkydiving_CR:bool=Get bIsSkyDiving()
    isEmoting_CR:bool=Get bIsPlayingEmote()
    isOnRail_CR:bool=Get bIsOnGrindRail()
    TacticalSprintingNoRB:double=Get TacticalSprintingNoRB() -> (TacticalSprintingNoRB:double)
    Rban:bool=Get bRigidBodyEnabled()
    jumpsprint:double=Get JumpingSpring() -> (JumpingSpring:double)
   ) -> (Pose) {
    alphaBoolBlend {blendOption=Linear}
    controlRigAssetReference {blueprintRigClass=&/M_MED_AutumnFern_CtrlRig.M_MED_AutumnFern_CtrlRig_C}
   }
   Save cached pose 'CR_Pose'(Pose=@n12) {cacheName=CR_Pose}
   n13 = RigidBody!(ComponentPose=Local To Component(LocalPose=Use cached pose 'CR_Pose'()), Alpha, OverrideWorldGravity:Vector="0,0,0", ExternalForce) -> (Pose) {
    bOverrideWorldGravity=true
    overridePhysicsAsset=&/M_MED_AutumnFern_Skydive_Physics
    simulationSpace=BaseBoneSpace
    alphaBoolBlend {blendInTime=2, blendOption=Linear, blendOutTime=0.5}
    alphaScaleBiasClamp {bMapRange=true}
    baseBoneRef {boneName=root}
    componentAppliedLinearAccClamp=<4000 4000 4000>
    simSpaceSettings {
     dampingAlpha=0
     maxAngularAcceleration=12
     maxAngularVelocity=5
     maxLinearAcceleration=900
     maxLinearVelocity=300
     velocityScaleZ=0.4
     worldAlpha=0.5
    }
   }
   n14 = "Transform (Modify) Bone - Bone: base_coatFront_r"(ComponentPose=@n13, bAlphaBoolEnabled=True, Translation="0,0,0", Rotation="0,-10,0", Scale="1,1,1") {
    alphaInputType=Bool
    rotationMode=BMM_Additive
    rotationSpace=BCS_BoneSpace
    translationMode=BMM_Additive
    translationSpace=BCS_BoneSpace
    alphaBoolBlend {blendInTime=0.1, blendOption=Linear, blendOutTime=0.1}
    boneToModify {boneName=base_coatFront_r}
    rotation {yaw=-10}
   }
   n15 = "Transform (Modify) Bone - Bone: base_coatFront_l"(ComponentPose=@n14, bAlphaBoolEnabled=True, Translation="0,0,0", Rotation="0,-15,0", Scale="1,1,1") {
    alphaInputType=Bool
    rotationMode=BMM_Additive
    rotationSpace=BCS_BoneSpace
    translationMode=BMM_Additive
    translationSpace=BCS_BoneSpace
    alphaBoolBlend {blendInTime=0.1, blendOption=Linear, blendOutTime=0.1}
    boneToModify {boneName=base_coatFront_l}
    rotation {yaw=-15}
   }
   n16 = "Transform (Modify) Bone - Bone: dyn_coatSide_02_l"!(ComponentPose=@n15, bAlphaBoolEnabled:bool=True, Translation="0,0,0", Rotation="-2.748,-19.526,-0.737", Scale="1,1,1") -> (Pose) {
    alphaInputType=Bool
    rotationMode=BMM_Additive
    rotationSpace=BCS_BoneSpace
    translationMode=BMM_Additive
    translationSpace=BCS_BoneSpace
    alphaBoolBlend {blendInTime=0.1, blendOption=Linear, blendOutTime=0.1}
    boneToModify {boneName=dyn_coatSide_02_l}
    rotation=<rot -2.748 -19.526 -0.737>
   }
   n17 = "Transform (Modify) Bone - Bone: dyn_coatFrontSide_02_r"!(ComponentPose=@n16, bAlphaBoolEnabled:bool=True, Translation="0,0,0", Rotation="-2.528,-7.093,-1.311", Scale="1,1,1") -> (Pose) {
    alphaInputType=Bool
    rotationMode=BMM_Additive
    rotationSpace=BCS_BoneSpace
    translationMode=BMM_Additive
    translationSpace=BCS_BoneSpace
    alphaBoolBlend {blendInTime=0.1, blendOption=Linear, blendOutTime=0.1}
    boneToModify {boneName=dyn_coatFrontSide_02_r}
    rotation=<rot -2.528 -7.093 -1.311>
   }
   n18 = "Transform (Modify) Bone - Bone: dyn_coatFront_03_r"(ComponentPose=@n17, bAlphaBoolEnabled=True, Translation="0,0,0", Rotation="-2.812,-26.123,-0.416", Scale="1,1,1") {
    alphaInputType=Bool
    rotationMode=BMM_Additive
    rotationSpace=BCS_BoneSpace
    translationMode=BMM_Additive
    translationSpace=BCS_BoneSpace
    alphaBoolBlend {blendInTime=0.1, blendOption=Linear, blendOutTime=0.1}
    boneToModify {boneName=dyn_coatFront_03_r}
    rotation=<rot -2.812 -26.123 -0.416>
   }
   n19 = "Transform (Modify) Bone - Bone: dyn_coatFrontSide_03_l"!(ComponentPose=@n18, bAlphaBoolEnabled:bool=True, Translation="0,0,0", Rotation="-2.833,-30.724,-0.194", Scale="1,1,1") -> (Pose) {
    alphaInputType=Bool
    rotationMode=BMM_Additive
    rotationSpace=BCS_BoneSpace
    translationMode=BMM_Additive
    translationSpace=BCS_BoneSpace
    alphaBoolBlend {blendInTime=0.1, blendOption=Linear, blendOutTime=0.1}
    boneToModify {boneName=dyn_coatFrontSide_03_l}
    rotation=<rot -2.833 -30.724 -0.194>
   }
   n20 = "Transform (Modify) Bone - Bone: deform_glute_r"(ComponentPose=@n19, bAlphaBoolEnabled=True, Translation="-10,0,0", Rotation="0,0,0", Scale="1,1,1") {
    alphaInputType=Bool
    rotationMode=BMM_Additive
    rotationSpace=BCS_BoneSpace
    translationMode=BMM_Additive
    translationSpace=BCS_BoneSpace
    alphaBoolBlend {blendInTime=0.3, blendOption=Linear, blendOutTime=0.3}
    boneToModify {boneName=deform_glute_r}
    translation {x=-10}
   }
   n21 = "Transform (Modify) Bone - Bone: deform_glute_l"(ComponentPose=@n20, bAlphaBoolEnabled=True, Translation="10,0,0", Rotation="0,0,0", Scale="1,1,1") {
    alphaInputType=Bool
    rotationMode=BMM_Additive
    rotationSpace=BCS_BoneSpace
    translationMode=BMM_Additive
    translationSpace=BCS_BoneSpace
    alphaBoolBlend {blendInTime=0.3, blendOption=Linear, blendOutTime=0.3}
    boneToModify {boneName=deform_glute_l}
    translation {x=10}
   }
   n22 = "Transform (Modify) Bone - Bone: base_coatBack_l"!(ComponentPose=@n21, bAlphaBoolEnabled:bool=True, Translation="0,0,0", Rotation="0,-15,0", Scale="1,1,1") -> (Pose) {
    alphaInputType=Bool
    rotationMode=BMM_Additive
    rotationSpace=BCS_BoneSpace
    translationMode=BMM_Additive
    translationSpace=BCS_BoneSpace
    alphaBoolBlend {blendInTime=0.3, blendOption=Linear, blendOutTime=0.3}
    boneToModify {boneName=base_coatBack_l}
    rotation {yaw=-15}
   }
   n23 = "Transform (Modify) Bone - Bone: base_coatBack_r"!(ComponentPose=@n22, bAlphaBoolEnabled:bool=True, Translation="0,0,0", Rotation="0,-15,0", Scale="1,1,1") -> (Pose) {
    alphaInputType=Bool
    rotationMode=BMM_Additive
    rotationSpace=BCS_BoneSpace
    translationMode=BMM_Additive
    translationSpace=BCS_BoneSpace
    alphaBoolBlend {blendInTime=0.3, blendOption=Linear, blendOutTime=0.3}
    boneToModify {boneName=base_coatBack_r}
    rotation {yaw=-15}
   }
   Save cached pose 'SkydivingPose'(Pose=Component To Local(ComponentPose=@n23)) {cacheName=SkydivingPose}
  ]}
 EventGraph {nodes [
   Event BlueprintUpdateAnimation()
   > n1 = Sequence!(execute) -> (then_0, then_1, then_2:Exec, then_3:Exec)
   n2 = Get bIsBaseballSliding()
   > Set RBANweight(RBANweight:double="To Float (Boolean)"(InBool=NOT Boolean(A=@n2))) -> (Output_Get:double)
   > Set isSkydivingOrEmoting(isSkydivingOrEmoting:bool=OR Boolean!(A=Get bIsSkyDiving(), B=Get bIsPlayingEmote(), C:bool=Get bIsKneeSliding()) -> (ReturnValue)) -> (Output_Get:bool)
   n3 = Get bIsCrouching()
   n4 = Get bIsCrouchSprinting()
   n5 = Get bIsBaseballSliding()
   > Set isIdle(isIdle:bool=NOT Boolean(A=OR Boolean(A=@n3, B=@n4, C=@n5))) -> (Output_Get:bool)
   n6 = NOT Boolean(A=Get bRigidBodyEnabled())
   n7 = Get bIsTacticalSprinting()
   > Set TacticalSprintingNoRB(TacticalSprintingNoRB:double="To Float (Boolean)"(InBool=AND Boolean(A=@n6, B=@n7))) -> (Output_Get:double)
   n8 = NOT Boolean(A=Get bIsTacticalSprinting())
   n9 = Get bIsSprinting()
   n10 = Get bIsJumping()
   > Set JumpingSpring(JumpingSpring:double="To Float (Boolean)"(InBool=AND Boolean!(A=@n10, B=@n9, C:bool=@n8) -> (ReturnValue))) -> (Output_Get:double)
   n11 = "Gravity Override State Function (Parameter Asset) (Nativized)"(
    execute=@n1.then_1
    CharacterPartAnimInstance=Self-Reference()
    GameTimeInSeconds=GetGameTimeInSeconds()
    DebugJointName=spine_05
    Parameters=&/M_MED_AutumnFern_Parameters
   )
   > Set Gravity Override(Gravity Override=@n11)
   Branch(execute=@n1.then_2, Condition=Get bIsInAir() -> (bIsInAir:bool))
   > Set afterJumping(afterJumping:bool=true) -> (Output_Get:bool)
   > Delay(Duration=0.6)
   > Set afterJumping(afterJumping=false)
  ]}
]
variables [
 afterJumping {type=boolean}
 Gravity Override {type=Vector}
 isIdle {type=boolean}
 isSkydivingOrEmoting {type=boolean}
 JumpingSpring {type=number}
 RBANweight {type=number}
 TacticalSprintingNoRB {type=number}
]
